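# Animation Blueprint for the M_MED_Alien_Robot player body cosmetic that adjusts its shoulder pads for certain player states.
# The AnimGraph copies the attached parent mesh's pose, then layers additive bone-space Transform Bone offsets with interpolated alphas.
# Each update, skydiving, diving or floating enables opposite X shifts on both pads.
# Targeting while crouched enables an extra move, rotation and scale on the right pad; otherwise each alpha is zero.
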AnimBlueprint /BRCosmetics/Characters/Player/Male/Medium/Bodies/M_MED_Alien_Robot/Meshes/M_MED_Alien_Robot_AnimBP
target_skeleton=&/M_MED_Alien_Robot_Skeleton
graphs [
 AnimGraph {nodes [
   n1 = Get Alpha() -> (Alpha:float)
   n2 = "Transform (Modify) Bone - Bone: shoulder_pad_l"(
    ComponentPose=Local To Component(LocalPose=Copy Pose From Mesh() {bUseAttachedParent=true})
    Alpha=@n1
    Translation="12,0,0"
    Rotation="0,0,0"
    Scale="1,1,1"
   ) {
    translationMode=BMM_Additive
    translationSpace=BCS_BoneSpace
    alphaBoolBlend {blendOption=Linear}
    alphaScaleBiasClamp {bInterpResult=true}
    boneToModify {boneName=shoulder_pad_l}
    translation {x=12}
   }
   n3 = "Transform (Modify) Bone - Bone: shoulder_pad_r"(ComponentPose=@n2, Alpha=@n1, Translation="-12,0,0", Rotation="0,0,0", Scale="1,1,1") {
    translationMode=BMM_Additive
    translationSpace=BCS_BoneSpace
    alphaBoolBlend {blendOption=Linear}
    alphaScaleBiasClamp {bInterpResult=true}
    boneToModify {boneName=shoulder_pad_r}
    translation {x=-12}
   }
   n4 = "Transform (Modify) Bone - Bone: shoulder_pad_r"(
    ComponentPose=@n3
    Alpha=Get Alpha Crouch() -> (Alpha Crouch:float)
    Translation="-0.084,0.358,3.093"
    Rotation="9.744,-11.028,-1.847"
    Scale="0.85,0.85,0.85"
   ) {
    rotationMode=BMM_Additive
    rotationSpace=BCS_BoneSpace
    scaleMode=BMM_Additive
    scaleSpace=BCS_BoneSpace
    translationMode=BMM_Additive
    translationSpace=BCS_BoneSpace
    alphaBoolBlend {blendOption=Linear}
    alphaScaleBiasClamp {bInterpResult=true}
    boneToModify {boneName=shoulder_pad_r}
    rotation=<rot 9.744 -11.028 -1.847>
    scale=<0.85 0.85 0.85>
    translation=<-0.084 0.358 3.093>
   }
   Output Pose(Result=Component To Local(ComponentPose=@n4)) {layerGroup=None}
  ]}
 EventGraph {nodes [
   Event BlueprintUpdateAnimation()
   > n1 = Sequence()
   > n2 = Branch(Condition=OR Boolean!(A=Get bIsSkyDiving(), B=Get bIsSkydiveDiveMode(), C:bool=Get bIsSkydiveFloating()) -> (ReturnValue))
   > Set Alpha(Alpha:float=Get Alpha_1() -> (Alpha_1:float)) -> (Output_Get:float)
   Set Alpha(execute=@n2.else, Alpha=0)
   n3 = Branch(execute=@n1.then_1, Condition=AND Boolean(A=Get bIsTargeting(), B=Get bIsCrouching()))
   > Set Alpha Crouch(Alpha Crouch:float=Get Alpha Crouch_0() -> (Alpha Crouch_0:float)) -> (Output_Get:float)
   Set Alpha Crouch(execute=@n3.else, Alpha Crouch=0)
  ]}
]
variables [
 Alpha {type=number, default=1}
 Alpha Crouch {type=number, default=1}
 Alpha Crouch_0 {type=number, default=1}
 Alpha_1 {type=number, default=1}
 Translation {type=Vector}
]
